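# Prototype Landmass erosion material that advects a packed height render target and injects rain, unlit with negative output allowed.
# R is velocity-advected, slope-smoothed height; B is backtraced through averaged neighboring VelocityRT samples, plus masked rain.
# The rain mask samples a 16x16 SubUV atlas (black placeholder) at a radially chosen slice, thresholds it by Radius and gates it by iteration index.
# MPC_Erosion supplies timestep, resolution, grid size, rain amount and index.
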Material /Landmass/Landscape/OldPrototype_BP/Mats/Advect/Custom/M_AdvectHeight_NiagaraForce
graph {
 properties {
  bAllowNegativeEmissiveColor=true
  blendMode=BLEND_AlphaComposite
  pixelDepthOffsetMode=PDOM_Legacy
  shadingModel=MSM_Unlit
  usedShadingModels=Unlit
 }
 nodes [
  n1 = TextureCoordinate()
  n2 = TextureCoordinate()
  n3 = Subtract(
   A=MaterialFunctionCall(Vector 2=Subtract(A=@n2) {constB=0.5}) {materialFunction=/Engine/Functions/Engine_MaterialFunctions02/Utility/VectorLength}.V2 Length
   B=ScalarParameter() {defaultValue=0.175, parameterName=Radial Max}
  ) {constB=0.175}
  n4 = LinearInterpolate(
   A=Add(
    A=TextureSample() {samplerType=SAMPLERTYPE_LinearColor, texture=/Engine/Functions/Engine_MaterialFunctions02/PivotPainter2/Black_1x1_EXR_Texture}.R
    B=ScalarParameter() {defaultValue=0.2, parameterName=Bottom offset}
   ) {constB=0.1}
   B=ScalarParameter() {defaultValue=15, parameterName=Z}
   Alpha=Saturate(None=Ceil(None=@n3))
  )
  AppendVector(A=@n1, B=@n4)
  Constant()
  n5 = Constant()
  n6 = CollectionParameter() {collection=/Landmass/Landscape/BlueprintBrushes/MPC/MPC_Erosion, parameterName=deltaT}
  n7 = MaterialFunctionCall() {materialFunction=/Landmass/Landscape/OldPrototype_BP/Mats/MF/PointSampleUVs}
  n8 = TextureSampleParameter2D(UVs=@n7.TexelCenterUVs) {parameterName=VelocityRT, texture=&Black}
  n9 = Divide(B=AppendVector(
    A=CollectionParameter() {collection=&MPC_Erosion, parameterName=ResX}
    B=CollectionParameter() {collection=&MPC_Erosion, parameterName=ResY}
   ))
  n10 = Constant2Vector() {r=1}
  n11 = TextureSampleParameter2D(UVs=Add(A=@n7.TexelCenterUVs, B=Multiply(A=@n9, B=@n10))) {parameterName=VelocityRT, texture=&Black}
  n12 = TextureSampleParameter2D(UVs=@n7.TexelCenterUVs) {parameterName=VelocityRT, texture=&Black}
  n13 = Constant2Vector() {g=1}
  n14 = TextureSampleParameter2D(UVs=Add(A=@n7.TexelCenterUVs, B=Multiply(A=@n9, B=@n13))) {parameterName=VelocityRT, texture=&Black}
  n15 = AppendVector(A=Divide(A=Add(A=@n8.R, B=@n11.R)), B=Divide(A=Add(A=@n12.G, B=@n14.G)))
  n16 = AppendVector(
   A=CollectionParameter() {collection=&MPC_Erosion, parameterName=ResX}
   B=CollectionParameter() {collection=&MPC_Erosion, parameterName=ResY}
  )
  n17 = CollectionParameter() {collection=&MPC_Erosion, parameterName=GridSize}
  n18 = Subtract(
   A=MaterialFunctionCall() {materialFunction=&PointSampleUVs}.TexelCenterUVs
   B=Divide(A=Multiply(A=@n6, B=@n15), B=Multiply(A=@n16, B=@n17))
  )
  n19 = Constant()
  n20 = Multiply(
   A=CollectionParameter() {collection=&MPC_Erosion, parameterName=RainAmount}
   B=CollectionParameter() {collection=&MPC_Erosion, parameterName=deltaT}
  )
  n21 = Constant() {r=16}
  n22 = Multiply(A=@n21, B=@n21)
  n23 = ComponentMask(None=MaterialFunctionCall!(
    Texture=TextureObject() {samplerType=SAMPLERTYPE_LinearColor, texture=&Black_1x1_EXR_Texture}
    UVs=@n1
    SubImages=Constant2Vector() {g=16, r=16}
    Frame=Multiply(A=@n4, B=@n22)
   ) {materialFunction=/Engine/Functions/Engine_MaterialFunctions03/Texturing/SubUV_Function}.RGB) {r=true}
  n24 = ScalarParameter() {defaultValue=0.03, parameterName=Radius}
  n25 = Multiply(A=Subtract(A=Saturate(None=@n23), B=@n24)) {constB=-1}
  n26 = Max(A=@n24) {constB=0.0001}
  n27 = Subtract(A=CollectionParameter() {collection=&MPC_Erosion, parameterName=idx}) {constB=100}
  n28 = Subtract(
   A=ScalarParameter() {defaultValue=300000, parameterName=Max Iterations}
   B=CollectionParameter() {collection=&MPC_Erosion, parameterName=idx}
  ) {constB=100}
  n29 = Multiply(A=Saturate(None=Divide(A=@n25, B=@n26)), B=Multiply(A=Saturate(None=@n27), B=Saturate(None=@n28)))
  MaterialFunctionCall(
   X=MaterialFunctionCall!(
    UVs=MaterialFunctionCall() {materialFunction=/Landmass/Landscape/OldPrototype_BP/Mats/MF/VelocityAdvect}.Advected UVs
    Height=TextureSampleParameter2D!(UVs, Level=@n5, Apply View MipBias) {mipValueMode=TMVM_MipLevel, parameterName=HeightRT, texture=&Black}.R
   ) {materialFunction=/Landmass/Landscape/OldPrototype_BP/Mats/MF/SlopeSmoothing}.Height
   Z=Add(
    A=TextureSampleParameter2D(UVs=@n18, Level=@n19) {mipValueMode=TMVM_MipLevel, parameterName=HeightRT, texture=&Black}.B
    B=Multiply(A=@n20, B=@n29)
   ) {constB=0}
  ) {materialFunction=&MakeFloat3}
  n30 = Constant() {r=0.5}
  n31 = CollectionParameter() {collection=&MPC_Erosion, parameterName=idx}
  n32 = ScalarParameter() {defaultValue=0.0001, parameterName=asdf}
  n33 = Multiply(
   A=VectorNoise(World Position=Multiply(A=@n31, B=@n32)) {noiseFunction=VNF_GradientALU}
   B=ScalarParameter() {defaultValue=0.025, parameterName=Noise}
  )
  SphereMask(A=TextureCoordinate(), B=ComponentMask(None=Add(A=@n30, B=@n33)) {g=true, r=true}) {attenuationRadius=0.075, hardnessPercent=0}
 ]
}
