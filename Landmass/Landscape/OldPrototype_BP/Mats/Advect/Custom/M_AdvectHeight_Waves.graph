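# Unlit render-target simulation material from the old Landmass prototype that advects water height and adds waves, using MPC_Erosion grid and timestep parameters.
# It traces UVs back along neighbor-averaged VelocityRT to sample HeightRT blue, then adds scaled WaveRT.
# A float3 packs slope-smoothed maximum of HeightRT green and offset waves, unchanged green, and that water value.
# Side branches compute a water-level cutoff and periodic blob injection.
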Material /Landmass/Landscape/OldPrototype_BP/Mats/Advect/Custom/M_AdvectHeight_Waves
graph {
 properties {
  bAllowNegativeEmissiveColor=true
  blendMode=BLEND_AlphaComposite
  pixelDepthOffsetMode=PDOM_Legacy
  shadingModel=MSM_Unlit
  usedShadingModels=Unlit
 }
 nodes [
  Constant()
  Constant() {r=0.5}
  n1 = TextureSampleParameter2D!(UVs, Level=Constant(), Apply View MipBias) {
   mipValueMode=TMVM_MipLevel
   parameterName=HeightRT
   samplerType=SAMPLERTYPE_LinearColor
   texture=/Engine/Functions/Engine_MaterialFunctions02/PivotPainter2/Black_1x1_EXR_Texture
  }
  n2 = MaterialFunctionCall() {materialFunction=/Landmass/Landscape/OldPrototype_BP/Mats/MF/PointSampleUVs}
  n3 = TextureSampleParameter2D!(UVs=@n2.TexelCenterUVs, Apply View MipBias) {
   parameterName=VelocityRT
   samplerType=SAMPLERTYPE_LinearColor
   texture=&Black_1x1_EXR_Texture
  }
  n4 = Divide(B=AppendVector(
    A=CollectionParameter() {collection=/Landmass/Landscape/BlueprintBrushes/MPC/MPC_Erosion, parameterName=ResX}
    B=CollectionParameter() {collection=&MPC_Erosion, parameterName=ResY}
   ))
  n5 = Constant2Vector() {r=1}
  n6 = TextureSampleParameter2D(UVs=Add(A=@n2.TexelCenterUVs, B=Multiply(A=@n4, B=@n5))) {
   parameterName=VelocityRT
   samplerType=SAMPLERTYPE_LinearColor
   texture=&Black_1x1_EXR_Texture
  }
  n7 = Add(A=@n3.R, B=@n6.R)
  n8 = TextureSampleParameter2D(UVs=@n2.TexelCenterUVs) {
   parameterName=VelocityRT
   samplerType=SAMPLERTYPE_LinearColor
   texture=&Black_1x1_EXR_Texture
  }
  n9 = Constant2Vector() {g=1}
  n10 = TextureSampleParameter2D(UVs=Add(A=@n2.TexelCenterUVs, B=Multiply(A=@n4, B=@n9))) {
   parameterName=VelocityRT
   samplerType=SAMPLERTYPE_LinearColor
   texture=&Black_1x1_EXR_Texture
  }
  n11 = Add(A=@n8.G, B=@n10.G)
  n12 = Multiply(
   A=CollectionParameter() {collection=&MPC_Erosion, parameterName=deltaT}
   B=AppendVector(A=Divide(A=@n7), B=Divide(A=@n11))
  )
  n13 = Multiply(
   A=AppendVector(
    A=CollectionParameter() {collection=&MPC_Erosion, parameterName=ResX}
    B=CollectionParameter() {collection=&MPC_Erosion, parameterName=ResY}
   )
   B=CollectionParameter() {collection=&MPC_Erosion, parameterName=GridSize}
  )
  n14 = TextureSampleParameter2D!(
   UVs=Subtract(
    A=MaterialFunctionCall() {materialFunction=&PointSampleUVs}.TexelCenterUVs
    B=Divide(A=@n12, B=@n13)
   )
   Level=Constant()
   Apply View MipBias
  ) {
   mipValueMode=TMVM_MipLevel
   parameterName=HeightRT
   samplerType=SAMPLERTYPE_LinearColor
   texture=&Black_1x1_EXR_Texture
  }
  n15 = TextureSampleParameter2D!(UVs, Apply View MipBias) {
   parameterName=WaveRT
   samplerType=SAMPLERTYPE_LinearColor
   texture=&Black_1x1_EXR_Texture
  }
  n16 = Add(A=@n14.B, B=Multiply(A=@n15.R, B=ScalarParameter() {parameterName=WaveHeight}))
  If(
   A=Add(A=@n1.R, B=@n16) {desc=Height of Water + Terrain}
   B=ScalarParameter() {defaultValue=100, parameterName=WL}
   "A > B"=@n16
   "A < B"=Constant()
  )
  n17 = TextureSampleParameter2D() {
   parameterName=WaveRT
   samplerType=SAMPLERTYPE_LinearColor
   texture=&Black_1x1_EXR_Texture
  }
  n18 = Add(A=@n17.R, B=ScalarParameter() {defaultValue=2000, parameterName=Wave Z})
  MaterialFunctionCall(
   X=MaterialFunctionCall!(
    UVs=MaterialFunctionCall() {materialFunction=/Landmass/Landscape/OldPrototype_BP/Mats/MF/VelocityAdvect}.Advected UVs
    Height=Max(A=@n1.G, B=@n18) {constB=0}
   ) {materialFunction=/Landmass/Landscape/OldPrototype_BP/Mats/MF/SlopeSmoothing}.Height
   Y=@n1.G
   Z=@n16
  ) {materialFunction=&MakeFloat3}
  n19 = TextureCoordinate()
  n20 = Multiply(
   A=SphereMask(
    A=ComponentMask(None=@n19) {g=true}
    B=ScalarParameter() {parameterName=LOC}
    Radius=ScalarParameter() {defaultValue=0.02, parameterName=R}
    Hardness=ScalarParameter() {defaultValue=0.9, parameterName=H}
   ) {hardnessPercent=0.5}
   B=ScalarParameter() {parameterName=Blob}
  )
  n21 = CollectionParameter() {collection=&MPC_Erosion, parameterName=idx}
  n22 = ScalarParameter() {defaultValue=2500, parameterName=Every N Frames}
  n23 = OneMinus(None=Saturate(None=Fmod(A=@n21, B=@n22)))
  Reroute(None=Add(A=@n14.B, B=Multiply(A=@n20, B=@n23)) {constB=0})
  Subtract(A=@n18, B=@n1.R)
 ]
}
