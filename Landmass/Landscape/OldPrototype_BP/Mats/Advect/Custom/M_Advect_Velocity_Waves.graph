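# Unlit simulation material that advects a velocity field from VelocityRT and adds wave forcing, part of an old Landmass erosion brush prototype.
# Semi-Lagrangian backtrace samples X and Y at half-texel staggered offsets, scaled by MPC_Erosion timestep, resolution and grid size.
# WaveRT height adds Y velocity, with sign flipping across the V midline.
# Separate branches, feeding no other node, compute face-averaged velocity and periodic curl-noise pulses along that midline.
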Material /Landmass/Landscape/OldPrototype_BP/Mats/Advect/Custom/M_Advect_Velocity_Waves
graph {
 properties {bAllowNegativeEmissiveColor=true, pixelDepthOffsetMode=PDOM_Legacy, shadingModel=MSM_Unlit, usedShadingModels=Unlit}
 nodes [
  n1 = TextureSampleParameter2D!(
   UVs=MaterialFunctionCall() {materialFunction=/Landmass/Landscape/OldPrototype_BP/Mats/MF/PointSampleUVs}.TexelCenterUVs
   Level=Constant()
   Apply View MipBias
  ) {
   constMipValue=0
   desc=New
   mipValueMode=TMVM_MipLevel
   parameterName=VelocityRT
   samplerType=SAMPLERTYPE_LinearColor
   texture=/Engine/Functions/Engine_MaterialFunctions02/PivotPainter2/Black_1x1_EXR_Texture
  }
  n2 = CollectionParameter() {collection=/Landmass/Landscape/BlueprintBrushes/MPC/MPC_Erosion, parameterName=ResX}
  n3 = CollectionParameter() {collection=&MPC_Erosion, parameterName=ResY}
  n4 = Multiply(A=Divide(B=AppendVector(A=@n2, B=@n3))) {constB=0.5}
  n5 = Constant2Vector() {g=1, r=-1}
  n6 = TextureSampleParameter2D(
   UVs=Add(
    A=MaterialFunctionCall() {materialFunction=&PointSampleUVs}.TexelCenterUVs
    B=Multiply(A=@n4, B=@n5)
   )
   Level=Constant()
  ) {
   constMipValue=0
   desc=New
   mipValueMode=TMVM_MipLevel
   parameterName=VelocityRT
   samplerType=SAMPLERTYPE_LinearColor
   texture=&Black_1x1_EXR_Texture
  }
  n7 = CollectionParameter() {collection=&MPC_Erosion, parameterName=deltaT}
  n8 = Multiply(A=AppendVector(A=@n1.R, B=@n6.G), B=@n7)
  n9 = Multiply(
   A=AppendVector(
    A=CollectionParameter() {collection=&MPC_Erosion, parameterName=ResX}
    B=CollectionParameter() {collection=&MPC_Erosion, parameterName=ResY}
   )
   B=CollectionParameter() {collection=&MPC_Erosion, parameterName=GridSize}
  )
  n10 = TextureSampleParameter2D(
   UVs=Subtract(
    A=MaterialFunctionCall() {materialFunction=&PointSampleUVs}.TexelCenterUVs
    B=Divide(A=@n8, B=@n9)
   )
   Level=Constant()
  ) {
   desc=New
   mipValueMode=TMVM_MipLevel
   parameterName=VelocityRT
   samplerType=SAMPLERTYPE_LinearColor
   texture=&Black_1x1_EXR_Texture
  }
  n11 = Constant2Vector() {g=-1, r=1}
  n12 = TextureSampleParameter2D(
   UVs=Add(
    A=MaterialFunctionCall() {materialFunction=&PointSampleUVs}.TexelCenterUVs
    B=Multiply(A=@n4, B=@n11)
   )
   Level=Constant()
  ) {
   constMipValue=0
   desc=New
   mipValueMode=TMVM_MipLevel
   parameterName=VelocityRT
   samplerType=SAMPLERTYPE_LinearColor
   texture=&Black_1x1_EXR_Texture
  }
  n13 = Multiply(A=@n7, B=AppendVector(A=@n12.R, B=@n1.G))
  n14 = TextureSampleParameter2D(
   UVs=Subtract(
    A=MaterialFunctionCall() {materialFunction=&PointSampleUVs}.TexelCenterUVs
    B=Divide(A=@n13, B=@n9)
   )
   Level=Constant()
  ) {
   desc=New
   mipValueMode=TMVM_MipLevel
   parameterName=VelocityRT
   samplerType=SAMPLERTYPE_LinearColor
   texture=&Black_1x1_EXR_Texture
  }
  n15 = TextureSampleParameter2D!(UVs, Apply View MipBias) {
   parameterName=WaveRT
   samplerType=SAMPLERTYPE_LinearColor
   texture=&Black_1x1_EXR_Texture
  }
  n16 = Multiply(A=@n15.R, B=ScalarParameter() {parameterName=WaveHeight})
  n17 = Constant3Vector() {constant=<rgba 0 1 0 0>}
  n18 = Subtract(A=ComponentMask(None=TextureCoordinate()) {g=true}) {constB=0.5}
  Add(
   A=MaterialFunctionCall(X=@n10.R, Y=@n14.G) {materialFunction=&MakeFloat3}.Result
   B=Multiply(A=Multiply(A=@n16, B=@n17), B=Sign(None=@n18))
  )
  n19 = TextureCoordinate()
  n20 = TextureSampleParameter2D(UVs=@n19) {
   desc=New
   parameterName=VelocityRT
   samplerType=SAMPLERTYPE_LinearColor
   texture=&Black_1x1_EXR_Texture
  }
  n21 = Divide(B=AppendVector(A=ScalarParameter() {defaultValue=1024, parameterName=SizeX}, B=ScalarParameter() {defaultValue=1024, parameterName=SizeY}))
  n22 = Constant2Vector() {r=1}
  n23 = TextureSampleParameter2D(UVs=Subtract(A=@n19, B=Multiply(A=@n21, B=@n22))) {
   desc=New
   parameterName=VelocityRT
   samplerType=SAMPLERTYPE_LinearColor
   texture=&Black_1x1_EXR_Texture
  }
  n24 = Constant2Vector() {g=1}
  n25 = TextureSampleParameter2D(UVs=Subtract(A=@n19, B=Multiply(A=@n21, B=@n24))) {
   desc=New
   parameterName=VelocityRT
   samplerType=SAMPLERTYPE_LinearColor
   texture=&Black_1x1_EXR_Texture
  }
  AppendVector(A=Divide(A=Add(A=@n20.R, B=@n23.R)), B=Divide(A=Add(A=@n20.G, B=@n25.G)))
  n26 = SphereMask(
   A=ComponentMask(None=TextureCoordinate()) {g=true}
   B=Constant() {r=0.5}
   Radius=ScalarParameter() {defaultValue=0.01, parameterName=R}
   Hardness=ScalarParameter() {defaultValue=0.9, parameterName=H}
  ) {hardnessPercent=0.5}
  n27 = TextureCoordinate()
  n28 = Divide(
   A=CollectionParameter() {collection=&MPC_Erosion, parameterName=idx}
   B=ScalarParameter() {defaultValue=123, parameterName=Noise Speed}
  )
  n29 = VectorNoise(World Position=Multiply(A=AppendVector(A=@n27, B=@n28), B=ScalarParameter() {defaultValue=2, parameterName=TilingN})) {noiseFunction=VNF_CurlALU}
  n30 = Fmod(
   A=CollectionParameter() {collection=&MPC_Erosion, parameterName=idx}
   B=ScalarParameter() {defaultValue=1500, parameterName=Every N Frames}
  )
  Multiply(
   A=Multiply(A=Multiply(A=@n26, B=@n29), B=ScalarParameter() {defaultValue=1500, parameterName=WaveNoise})
   B=OneMinus(None=Saturate(None=@n30))
  )
  Multiply(A=@n15.R, B=@n15.R)
  TextureCoordinate()
  TextureCoordinate()
  TextureSampleParameter2D() {desc=New, parameterName=VelocityRT, texture=&Black_1x1_EXR_Texture}
 ]
}
